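# Post-process style AnimBP for the female CosmicSquatter player body, layering wing, tail and calf-feather motion onto the parent mesh pose.
# Copied pose feeds a Control Rig driven by sprint/emote/crouch, then additive bone tweaks for parachute, kneeslide, frontend and foot proximity.
# RigidBody then simulates the tail, and per-bone AnimDynamics springs the wings and feathers.
# EventGraph derives alphas, using foot-to-foot and right-foot-to-tail distances for tail avoidance outside emotes.
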AnimBlueprint /Game/Characters/Player/Female/Medium/Bodies/F_MED_CosmicSquatter/Meshes/F_MED_CosmicSquatter_AnimBP
target_skeleton=&/F_MED_CosmicSquatter_Skeleton
graphs [
 AnimGraph {nodes [
   n1 = Control Rig!(
    Source=Copy Pose From Mesh() {bUseAttachedParent=true}
    Alpha
    sprint:double=Get sprint() -> (sprint:double)
    emote:double=Get emote() -> (emote:double)
    crouch:double=Get Crouch() -> (Crouch:double)
   ) -> (Pose) {
    alphaBoolBlend {blendOption=Linear}
    controlRigAssetReference {blueprintRigClass=&/F_MED_CosmicSquatter_CtrlRig.F_MED_CosmicSquatter_CtrlRig_C}
   }
   n2 = Get bIsParachuteOpened()
   n3 = "Transform (Modify) Bone - Bone: dyn_wing_upper_l_twist_MSTR"!(
    ComponentPose=Local To Component(LocalPose=@n1)
    bAlphaBoolEnabled:bool=@n2
    Translation="0,0,0"
    Rotation="-0.010,0.039,15.000"
    Scale="1,1,1"
   ) -> (Pose) {
    alphaInputType=Bool
    rotationMode=BMM_Additive
    rotationSpace=BCS_BoneSpace
    alphaBoolBlend {blendInTime=0.1, blendOption=Linear, blendOutTime=0.1}
    boneToModify {boneName=dyn_wing_upper_l_twist_MSTR}
    rotation=<rot -0.01 0.039 15>
   }
   n4 = "Transform (Modify) Bone - Bone: dyn_wing_upper_r_twist_MSTR"!(ComponentPose=@n3, bAlphaBoolEnabled:bool=@n2, Translation="0,0,0", Rotation="0,0,15", Scale="1,1,1") -> (Pose) {
    alphaInputType=Bool
    rotationMode=BMM_Additive
    rotationSpace=BCS_BoneSpace
    alphaBoolBlend {blendInTime=0.1, blendOption=Linear, blendOutTime=0.1}
    boneToModify {boneName=dyn_wing_upper_r_twist_MSTR}
    rotation {roll=15}
   }
   n5 = Get frontend() -> (frontend:double)
   n6 = "Transform (Modify) Bone - Bone: dyn_wing_upper_l_twist_MSTR"!(ComponentPose=@n4, Alpha:float=@n5, Translation="0,0,0", Rotation="5,0,-5", Scale="1,1,1") -> (Pose) {
    rotationMode=BMM_Additive
    rotationSpace=BCS_BoneSpace
    alphaBoolBlend {blendInTime=0.1, blendOption=Linear, blendOutTime=0.1}
    boneToModify {boneName=dyn_wing_upper_l_twist_MSTR}
    rotation {pitch=5, roll=-5}
   }
   n7 = "Transform (Modify) Bone - Bone: dyn_wing_upper_l_MSTR"(ComponentPose=@n6, Alpha=@n5, Translation="0,-2,-1", Rotation="0,0,0", Scale="1,1,1") {
    rotationMode=BMM_Additive
    rotationSpace=BCS_BoneSpace
    translationMode=BMM_Additive
    translationSpace=BCS_BoneSpace
    alphaBoolBlend {blendInTime=0.1, blendOption=Linear, blendOutTime=0.1}
    boneToModify {boneName=dyn_wing_upper_l_MSTR}
    translation {y=-2, z=-1}
   }
   n8 = "Transform (Modify) Bone - Bone: dyn_wing_upper_r_twist_MSTR"!(ComponentPose=@n7, Alpha:float=@n5, Translation="0,0,0", Rotation="-5,0,5", Scale="1,1,1") -> (Pose) {
    rotationMode=BMM_Additive
    rotationSpace=BCS_BoneSpace
    alphaBoolBlend {blendInTime=0.1, blendOption=Linear, blendOutTime=0.1}
    boneToModify {boneName=dyn_wing_upper_r_twist_MSTR}
    rotation {pitch=-5, roll=5}
   }
   n9 = "Transform (Modify) Bone - Bone: dyn_wing_upper_r_MSTR"(ComponentPose=@n8, Alpha=@n5, Translation="0,2,1", Rotation="0,0,0", Scale="1,1,1") {
    rotationMode=BMM_Additive
    rotationSpace=BCS_BoneSpace
    translationMode=BMM_Additive
    translationSpace=BCS_BoneSpace
    alphaBoolBlend {blendInTime=0.1, blendOption=Linear, blendOutTime=0.1}
    boneToModify {boneName=dyn_wing_upper_r_MSTR}
    translation {y=2, z=1}
   }
   n10 = "Transform (Modify) Bone - Bone: dyn_tail_MSTR"!(ComponentPose=@n9, bAlphaBoolEnabled:bool=Get bIsKneeSliding(), Translation="0,0,5", Rotation="0,0,60", Scale="1,1,1") -> (Pose) {
    alphaInputType=Bool
    rotationMode=BMM_Additive
    translationMode=BMM_Additive
    alphaBoolBlend {blendInTime=0.1, blendOption=Linear, blendOutTime=0.1}
    alphaScaleBiasClamp {bInterpResult=true}
    boneToModify {boneName=dyn_tail_MSTR}
    rotation {roll=60}
    translation {z=5}
   }
   n11 = "Transform (Modify) Bone - Bone: dyn_tail_MSTR"(ComponentPose=@n10, bAlphaBoolEnabled=Get bIsParachuteOpened(), Translation="0,-7,0", Rotation="0,0,0", Scale="1,1,1") {
    alphaInputType=Bool
    rotationMode=BMM_Additive
    translationMode=BMM_Additive
    translationSpace=BCS_ParentBoneSpace
    alphaBoolBlend {blendInTime=0.5, blendOption=Linear, blendOutTime=0.5}
    alphaScaleBiasClamp {bInterpResult=true}
    boneToModify {boneName=dyn_tail_MSTR}
    translation {y=-7}
   }
   n12 = "Transform (Modify) Bone - Bone: dyn_tail_MSTR"!(ComponentPose=@n11, Alpha:float=Get feetCloseToTail() -> (feetCloseToTail:double), Translation="-7,-16,0", Rotation="0,-40,0", Scale="1,1,1") -> (Pose) {
    rotationMode=BMM_Additive
    rotationSpace=BCS_BoneSpace
    translationMode=BMM_Additive
    translationSpace=BCS_BoneSpace
    alphaBoolBlend {blendOption=Linear}
    boneToModify {boneName=dyn_tail_MSTR}
    rotation {yaw=-40}
    translation=<-7 -16>
   }
   n13 = RigidBody!(ComponentPose=@n12, Alpha, OverrideWorldGravity:Vector="0,0,0", ExternalForce) -> (Pose) {
    bOverrideWorldGravity=true
    overridePhysicsAsset=&/F_MED_CosmicSquatter_Tail_Physics
    simulationSpace=BaseBoneSpace
    alphaBoolBlend {blendOption=Linear}
    baseBoneRef {boneName=pelvis}
    componentAppliedLinearAccClamp=<3000 3000 3000>
    componentLinearAccScale=<0.3 0.2 0.4>
    componentLinearVelScale=<0.1 0.1 0.2>
    simSpaceSettings {
     maxAngularAcceleration=100
     maxAngularVelocity=3
     maxLinearAcceleration=700
     maxLinearVelocity=500
     velocityScaleZ=0
     worldAlpha=1
     externalLinearDragV=<0.2 0.2>
    }
   }
   n14 = Anim Dynamics - Bone: dyn_wing_upper_r_02(ComponentPose=@n13, Alpha=0.3, ExternalForce="0,0,0") {
    angularDampingOverride=0.99
    angularSpringConstant=5000
    bAngularSpring=true
    bGravityOverrideInSimSpace=true
    bOverrideAngularDamping=true
    simulationSpace=RootRelative
    alphaBoolBlend {blendOption=Linear}
    boundBone {boneName=dyn_wing_upper_r_02}
    componentAppliedLinearAccClamp=<3500 3500 3500>
    componentLinearAccScale=<0.5 0.5 0.6>
    componentLinearVelScale=<0.5 0.5 0.6>
    gravityOverride {z=-980}
    physicsBodyDefinitions [{
      sphereCollisionRadius=0
      boundBone {boneName=dyn_wing_upper_r_02}
      boxExtents=<16 16 16>
      constraintSetup {angularLimitsMax=<1 25 10>, angularLimitsMin=<-1 -10 -25>, angularTarget=<1 0 0>}
      localJointOffset=<1 0 0>
     }]
    relativeSpaceBone {boneName=root}
   }
   n15 = Anim Dynamics - Bone: dyn_wing_upper_r_03(ComponentPose=@n14, Alpha=0.3, ExternalForce="0,0,0") {
    angularDampingOverride=0.99
    angularSpringConstant=5000
    bAngularSpring=true
    bGravityOverrideInSimSpace=true
    bOverrideAngularDamping=true
    simulationSpace=RootRelative
    alphaBoolBlend {blendOption=Linear}
    boundBone {boneName=dyn_wing_upper_r_03}
    componentAppliedLinearAccClamp=<3500 3500 3500>
    componentLinearAccScale=<0.5 0.5 0.6>
    componentLinearVelScale=<0.5 0.5 0.6>
    gravityOverride {z=-980}
    physicsBodyDefinitions [{
      sphereCollisionRadius=0
      boundBone {boneName=dyn_wing_upper_r_03}
      boxExtents=<16 16 16>
      constraintSetup {angularLimitsMax=<1 25 10>, angularLimitsMin=<-1 -10 -25>, angularTarget=<1 0 0>}
      localJointOffset=<1 0 0>
     }]
    relativeSpaceBone {boneName=root}
   }
   n16 = Anim Dynamics - Bone: dyn_wing_upper_r_01(ComponentPose=@n15, Alpha=0.3, ExternalForce="0,0,0") {
    angularDampingOverride=0.99
    angularSpringConstant=5000
    bAngularSpring=true
    bGravityOverrideInSimSpace=true
    bOverrideAngularDamping=true
    simulationSpace=RootRelative
    alphaBoolBlend {blendOption=Linear}
    boundBone {boneName=dyn_wing_upper_r_01}
    componentAppliedLinearAccClamp=<3500 3500 3500>
    componentLinearAccScale=<0.5 0.5 0.6>
    componentLinearVelScale=<0.5 0.5 0.6>
    gravityOverride {z=-980}
    physicsBodyDefinitions [{
      sphereCollisionRadius=0
      boundBone {boneName=dyn_wing_upper_r_01}
      boxExtents=<16 16 16>
      constraintSetup {angularLimitsMax=<1 25 10>, angularLimitsMin=<-1 -5 -25>, angularTarget=<1 0 0>}
      localJointOffset=<1 0 0>
     }]
    relativeSpaceBone {boneName=root}
   }
   n17 = Anim Dynamics - Bone: dyn_wing_upper_r_04(ComponentPose=@n16, Alpha=0.3, ExternalForce="0,0,0") {
    angularDampingOverride=0.99
    angularSpringConstant=5000
    bAngularSpring=true
    bGravityOverrideInSimSpace=true
    bOverrideAngularDamping=true
    simulationSpace=RootRelative
    alphaBoolBlend {blendOption=Linear}
    boundBone {boneName=dyn_wing_upper_r_04}
    componentAppliedLinearAccClamp=<3500 3500 3500>
    componentLinearAccScale=<0.5 0.5 0.6>
    componentLinearVelScale=<0.5 0.5 0.6>
    gravityOverride {z=-980}
    physicsBodyDefinitions [{
      sphereCollisionRadius=0
      boundBone {boneName=dyn_wing_upper_r_04}
      boxExtents=<16 16 16>
      constraintSetup {angularLimitsMax=<1 25 10>, angularLimitsMin=<-1 0 -25>, angularTarget=<1 0 0>}
      localJointOffset=<1 0 0>
     }]
    relativeSpaceBone {boneName=root}
   }
   n18 = Anim Dynamics - Bone: dyn_wing_upper_r_05(ComponentPose=@n17, Alpha=0.3, ExternalForce="0,0,0") {
    angularDampingOverride=0.99
    angularSpringConstant=5000
    bAngularSpring=true
    bGravityOverrideInSimSpace=true
    bOverrideAngularDamping=true
    simulationSpace=RootRelative
    alphaBoolBlend {blendOption=Linear}
    boundBone {boneName=dyn_wing_upper_r_05}
    componentAppliedLinearAccClamp=<3500 3500 3500>
    componentLinearAccScale=<0.5 0.5 0.6>
    componentLinearVelScale=<0.5 0.5 0.6>
    gravityOverride {z=-980}
    physicsBodyDefinitions [{
      sphereCollisionRadius=0
      boundBone {boneName=dyn_wing_upper_r_05}
      boxExtents=<16 16 16>
      constraintSetup {angularLimitsMax=<1 25 10>, angularLimitsMin=<-1 0 -25>, angularTarget=<1 0 0>}
      localJointOffset=<1 0 0>
     }]
    relativeSpaceBone {boneName=root}
   }
   n19 = Anim Dynamics - Bone: dyn_wing_upper_l_02(ComponentPose=@n18, Alpha=0.3, ExternalForce="0,0,0") {
    angularDampingOverride=0.99
    angularSpringConstant=5000
    bAngularSpring=true
    bGravityOverrideInSimSpace=true
    bOverrideAngularDamping=true
    simulationSpace=RootRelative
    alphaBoolBlend {blendOption=Linear}
    boundBone {boneName=dyn_wing_upper_l_02}
    componentAppliedLinearAccClamp=<3500 3500 3500>
    componentLinearAccScale=<0.5 0.5 0.6>
    componentLinearVelScale=<0.5 0.5 0.6>
    gravityOverride {z=-980}
    physicsBodyDefinitions [{
      sphereCollisionRadius=0
      boundBone {boneName=dyn_wing_upper_l_02}
      boxExtents=<16 16 16>
      constraintSetup {angularLimitsMax=<1 25 10>, angularLimitsMin=<-1 -10 -25>, angularTarget=<1 0 0>}
      localJointOffset=<-1 0 0>
     }]
    relativeSpaceBone {boneName=root}
   }
   n20 = Anim Dynamics - Bone: dyn_wing_upper_l_03(ComponentPose=@n19, Alpha=0.3, ExternalForce="0,0,0") {
    angularDampingOverride=0.99
    angularSpringConstant=5000
    bAngularSpring=true
    bGravityOverrideInSimSpace=true
    bOverrideAngularDamping=true
    simulationSpace=RootRelative
    alphaBoolBlend {blendOption=Linear}
    boundBone {boneName=dyn_wing_upper_l_03}
    componentAppliedLinearAccClamp=<3500 3500 3500>
    componentLinearAccScale=<0.5 0.5 0.6>
    componentLinearVelScale=<0.5 0.5 0.6>
    gravityOverride {z=-980}
    physicsBodyDefinitions [{
      sphereCollisionRadius=0
      boundBone {boneName=dyn_wing_upper_l_03}
      boxExtents=<16 16 16>
      constraintSetup {angularLimitsMax=<1 25 10>, angularLimitsMin=<-1 -10 -25>, angularTarget=<1 0 0>}
      localJointOffset=<-1 0 0>
     }]
    relativeSpaceBone {boneName=root}
   }
   n21 = Anim Dynamics - Bone: dyn_wing_upper_l_01(ComponentPose=@n20, Alpha=0.3, ExternalForce="0,0,0") {
    angularDampingOverride=0.99
    angularSpringConstant=5000
    bAngularSpring=true
    bGravityOverrideInSimSpace=true
    bOverrideAngularDamping=true
    simulationSpace=RootRelative
    alphaBoolBlend {blendOption=Linear}
    boundBone {boneName=dyn_wing_upper_l_01}
    componentAppliedLinearAccClamp=<3500 3500 3500>
    componentLinearAccScale=<0.5 0.5 0.6>
    componentLinearVelScale=<0.5 0.5 0.6>
    gravityOverride {z=-980}
    physicsBodyDefinitions [{
      sphereCollisionRadius=0
      boundBone {boneName=dyn_wing_upper_l_01}
      boxExtents=<16 16 16>
      constraintSetup {angularLimitsMax=<1 25 10>, angularLimitsMin=<-1 -5 -25>, angularTarget=<1 0 0>}
      localJointOffset=<-1 0 0>
     }]
    relativeSpaceBone {boneName=root}
   }
   n22 = Anim Dynamics - Bone: dyn_wing_upper_l_04(ComponentPose=@n21, Alpha=0.3, ExternalForce="0,0,0") {
    angularDampingOverride=0.99
    angularSpringConstant=5000
    bAngularSpring=true
    bGravityOverrideInSimSpace=true
    bOverrideAngularDamping=true
    simulationSpace=RootRelative
    alphaBoolBlend {blendOption=Linear}
    boundBone {boneName=dyn_wing_upper_l_04}
    componentAppliedLinearAccClamp=<3500 3500 3500>
    componentLinearAccScale=<0.5 0.5 0.6>
    componentLinearVelScale=<0.5 0.5 0.6>
    gravityOverride {z=-980}
    physicsBodyDefinitions [{
      sphereCollisionRadius=0
      boundBone {boneName=dyn_wing_upper_l_04}
      boxExtents=<16 16 16>
      constraintSetup {angularLimitsMax=<1 25 10>, angularLimitsMin=<-1 0 -25>, angularTarget=<1 0 0>}
      localJointOffset=<-1 0 0>
     }]
    relativeSpaceBone {boneName=root}
   }
   n23 = Anim Dynamics - Bone: dyn_wing_upper_l_05(ComponentPose=@n22, Alpha=0.3, ExternalForce="0,0,0") {
    angularDampingOverride=0.99
    angularSpringConstant=5000
    bAngularSpring=true
    bGravityOverrideInSimSpace=true
    bOverrideAngularDamping=true
    simulationSpace=RootRelative
    alphaBoolBlend {blendOption=Linear}
    boundBone {boneName=dyn_wing_upper_l_05}
    componentAppliedLinearAccClamp=<3500 3500 3500>
    componentLinearAccScale=<0.5 0.5 0.6>
    componentLinearVelScale=<0.5 0.5 0.6>
    gravityOverride {z=-980}
    physicsBodyDefinitions [{
      sphereCollisionRadius=0
      boundBone {boneName=dyn_wing_upper_l_05}
      boxExtents=<16 16 16>
      constraintSetup {angularLimitsMax=<1 25 10>, angularLimitsMin=<-1 0 -25>, angularTarget=<1 0 0>}
      localJointOffset=<-1 0 0>
     }]
    relativeSpaceBone {boneName=root}
   }
   n24 = Anim Dynamics - Bone: dyn_calf_feather_01(ComponentPose=@n23, Alpha=0.5, ExternalForce="0,0,0") {
    angularDampingOverride=0.9999
    angularSpringConstant=75
    bAngularSpring=true
    bGravityOverrideInSimSpace=true
    bOverrideAngularDamping=true
    gravityScale=0.8
    simulationSpace=BoneRelative
    alphaBoolBlend {blendOption=Linear}
    boundBone {boneName=dyn_calf_feather_01}
    componentAppliedLinearAccClamp=<3500 3500 3500>
    componentLinearAccScale=<1.5 1.5 1.25>
    componentLinearVelScale=<1.5 1.5 1.25>
    gravityOverride {z=-980}
    physicsBodyDefinitions [{
      sphereCollisionRadius=0
      boundBone {boneName=dyn_calf_feather_01}
      boxExtents=<17 17 10>
      constraintSetup {angularLimitsMax=<5 3 15>, angularLimitsMin=<-5 -15 -15>, angularTarget=<1 0 0>}
      localJointOffset=<3 0 0>
     }]
    relativeSpaceBone {boneName=root}
   }
   n25 = Anim Dynamics - Bone: dyn_calf_feather_02(ComponentPose=@n24, Alpha=0.5, ExternalForce="0,0,0") {
    angularDampingOverride=0.9999
    angularSpringConstant=75
    bAngularSpring=true
    bGravityOverrideInSimSpace=true
    bOverrideAngularDamping=true
    gravityScale=0.8
    simulationSpace=BoneRelative
    alphaBoolBlend {blendOption=Linear}
    boundBone {boneName=dyn_calf_feather_02}
    componentAppliedLinearAccClamp=<3500 3500 3500>
    componentLinearAccScale=<1.5 1.5 1.25>
    componentLinearVelScale=<1.5 1.5 1.25>
    gravityOverride {z=-980}
    physicsBodyDefinitions [{
      sphereCollisionRadius=0
      boundBone {boneName=dyn_calf_feather_02}
      boxExtents=<17 17 10>
      constraintSetup {angularLimitsMax=<5 3 15>, angularLimitsMin=<-5 -15 -15>, angularTarget=<1 0 0>}
      localJointOffset=<4 0.5 0>
     }]
    relativeSpaceBone {boneName=root}
   }
   n26 = Anim Dynamics - Bone: dyn_calf_feather_03(ComponentPose=@n25, Alpha=0.5, ExternalForce="0,0,0") {
    angularDampingOverride=0.9999
    angularSpringConstant=75
    bAngularSpring=true
    bGravityOverrideInSimSpace=true
    bOverrideAngularDamping=true
    gravityScale=0.8
    simulationSpace=BoneRelative
    alphaBoolBlend {blendOption=Linear}
    boundBone {boneName=dyn_calf_feather_03}
    componentAppliedLinearAccClamp=<3500 3500 3500>
    componentLinearAccScale=<1.5 1.5 1.25>
    componentLinearVelScale=<1.5 1.5 1.25>
    gravityOverride {z=-980}
    physicsBodyDefinitions [{
      sphereCollisionRadius=0
      boundBone {boneName=dyn_calf_feather_03}
      boxExtents=<17 17 10>
      constraintSetup {angularLimitsMax=<5 3 15>, angularLimitsMin=<-5 -15 -15>, angularTarget=<1 0 0>}
      localJointOffset=<3 1 0>
     }]
    relativeSpaceBone {boneName=root}
   }
   n27 = Anim Dynamics - Bone: dyn_calf_feather_04(ComponentPose=@n26, Alpha=0.5, ExternalForce="0,0,0") {
    angularDampingOverride=0.9999
    angularSpringConstant=75
    bAngularSpring=true
    bGravityOverrideInSimSpace=true
    bOverrideAngularDamping=true
    gravityScale=0.8
    simulationSpace=BoneRelative
    alphaBoolBlend {blendOption=Linear}
    boundBone {boneName=dyn_calf_feather_04}
    componentAppliedLinearAccClamp=<3500 3500 3500>
    componentLinearAccScale=<1.5 1.5 1.25>
    componentLinearVelScale=<1.5 1.5 1.25>
    gravityOverride {z=-980}
    physicsBodyDefinitions [{
      sphereCollisionRadius=0
      boundBone {boneName=dyn_calf_feather_04}
      boxExtents=<17 17 10>
      constraintSetup {angularLimitsMax=<5 3 15>, angularLimitsMin=<-5 -15 -15>, angularTarget=<1 0 0>}
      localJointOffset=<4 0 0>
     }]
    relativeSpaceBone {boneName=root}
   }
   n28 = Anim Dynamics - Bone: dyn_calf_feather_05(ComponentPose=@n27, Alpha=0.5, ExternalForce="0,0,0") {
    angularDampingOverride=0.9999
    angularSpringConstant=75
    bAngularSpring=true
    bGravityOverrideInSimSpace=true
    bOverrideAngularDamping=true
    gravityScale=0.8
    simulationSpace=BoneRelative
    alphaBoolBlend {blendOption=Linear}
    boundBone {boneName=dyn_calf_feather_05}
    componentAppliedLinearAccClamp=<3500 3500 3500>
    componentLinearAccScale=<1.5 1.5 1.25>
    componentLinearVelScale=<1.5 1.5 1.25>
    gravityOverride {z=-980}
    physicsBodyDefinitions [{
      sphereCollisionRadius=0
      boundBone {boneName=dyn_calf_feather_05}
      boxExtents=<17 17 10>
      constraintSetup {angularLimitsMax=<5 3 15>, angularLimitsMin=<-5 -15 -15>, angularTarget=<1 0 0>}
      localJointOffset=<3 0 0>
     }]
    relativeSpaceBone {boneName=root}
   }
   n29 = Anim Dynamics - Bone: dyn_calf_feather_06(ComponentPose=@n28, Alpha=0.5, ExternalForce="0,0,0") {
    angularDampingOverride=0.9999
    angularSpringConstant=75
    bAngularSpring=true
    bGravityOverrideInSimSpace=true
    bOverrideAngularDamping=true
    gravityScale=0.8
    simulationSpace=BoneRelative
    alphaBoolBlend {blendOption=Linear}
    boundBone {boneName=dyn_calf_feather_06}
    componentAppliedLinearAccClamp=<3500 3500 3500>
    componentLinearAccScale=<1.5 1.5 1.25>
    componentLinearVelScale=<1.5 1.5 1.25>
    gravityOverride {z=-980}
    physicsBodyDefinitions [{
      sphereCollisionRadius=0
      boundBone {boneName=dyn_calf_feather_06}
      boxExtents=<17 17 10>
      constraintSetup {angularLimitsMax=<5 3 15>, angularLimitsMin=<-5 -15 -15>, angularTarget=<1 0 0>}
      localJointOffset=<4 0.5 0>
     }]
    relativeSpaceBone {boneName=root}
   }
   n30 = Anim Dynamics - Bone: dyn_calf_feather_07(ComponentPose=@n29, Alpha=0.5, ExternalForce="0,0,0") {
    angularDampingOverride=0.9999
    angularSpringConstant=75
    bAngularSpring=true
    bGravityOverrideInSimSpace=true
    bOverrideAngularDamping=true
    gravityScale=0.8
    simulationSpace=BoneRelative
    alphaBoolBlend {blendOption=Linear}
    boundBone {boneName=dyn_calf_feather_07}
    componentAppliedLinearAccClamp=<3500 3500 3500>
    componentLinearAccScale=<1.5 1.5 1.25>
    componentLinearVelScale=<1.5 1.5 1.25>
    gravityOverride {z=-980}
    physicsBodyDefinitions [{
      sphereCollisionRadius=0
      boundBone {boneName=dyn_calf_feather_07}
      boxExtents=<17 17 10>
      constraintSetup {angularLimitsMax=<5 3 15>, angularLimitsMin=<-5 -15 -15>, angularTarget=<1 0 0>}
      localJointOffset=<3 1 0>
     }]
    relativeSpaceBone {boneName=root}
   }
   n31 = Anim Dynamics - Bone: dyn_calf_feather_08(ComponentPose=@n30, Alpha=0.5, ExternalForce="0,0,0") {
    angularDampingOverride=0.9999
    angularSpringConstant=75
    bAngularSpring=true
    bGravityOverrideInSimSpace=true
    bOverrideAngularDamping=true
    gravityScale=0.8
    simulationSpace=BoneRelative
    alphaBoolBlend {blendOption=Linear}
    boundBone {boneName=dyn_calf_feather_08}
    componentAppliedLinearAccClamp=<3500 3500 3500>
    componentLinearAccScale=<1.5 1.5 1.25>
    componentLinearVelScale=<1.5 1.5 1.25>
    gravityOverride {z=-980}
    physicsBodyDefinitions [{
      sphereCollisionRadius=0
      boundBone {boneName=dyn_calf_feather_08}
      boxExtents=<17 17 10>
      constraintSetup {angularLimitsMax=<5 3 15>, angularLimitsMin=<-5 -15 -15>, angularTarget=<1 0 0>}
      localJointOffset=<4 0 0>
     }]
    relativeSpaceBone {boneName=root}
   }
   n32 = Anim Dynamics - Bone: dyn_calf_feather_l_01(ComponentPose=@n31, Alpha=0.5, ExternalForce="0,0,0") {
    angularDampingOverride=0.9999
    angularSpringConstant=75
    bAngularSpring=true
    bGravityOverrideInSimSpace=true
    bOverrideAngularDamping=true
    gravityScale=0.8
    simulationSpace=BoneRelative
    alphaBoolBlend {blendOption=Linear}
    boundBone {boneName=dyn_calf_feather_l_01}
    componentAppliedLinearAccClamp=<3500 3500 3500>
    componentLinearAccScale=<1.5 1.5 1.25>
    componentLinearVelScale=<1.5 1.5 1.25>
    gravityOverride {z=-980}
    physicsBodyDefinitions [{
      sphereCollisionRadius=0
      boundBone {boneName=dyn_calf_feather_l_01}
      boxExtents=<17 17 10>
      constraintSetup {angularLimitsMax=<5 3 15>, angularLimitsMin=<-5 -15 -15>, angularTarget=<1 0 0>}
      localJointOffset=<3 -0.5 0>
     }]
    relativeSpaceBone {boneName=root}
   }
   n33 = Anim Dynamics - Bone: dyn_calf_feather_l_02(ComponentPose=@n32, Alpha=0.5, ExternalForce="0,0,0") {
    angularDampingOverride=0.9999
    angularSpringConstant=75
    bAngularSpring=true
    bGravityOverrideInSimSpace=true
    bOverrideAngularDamping=true
    gravityScale=0.8
    simulationSpace=BoneRelative
    alphaBoolBlend {blendOption=Linear}
    boundBone {boneName=dyn_calf_feather_l_02}
    componentAppliedLinearAccClamp=<3500 3500 3500>
    componentLinearAccScale=<1.5 1.5 1.25>
    componentLinearVelScale=<1.5 1.5 1.25>
    gravityOverride {z=-980}
    physicsBodyDefinitions [{
      sphereCollisionRadius=0
      boundBone {boneName=dyn_calf_feather_l_02}
      boxExtents=<17 17 10>
      constraintSetup {angularLimitsMax=<5 3 15>, angularLimitsMin=<-5 -15 -15>, angularTarget=<1 0 0>}
      localJointOffset=<4 0.5 0>
     }]
    relativeSpaceBone {boneName=root}
   }
   n34 = Anim Dynamics - Bone: dyn_calf_feather_l_03(ComponentPose=@n33, Alpha=0.5, ExternalForce="0,0,0") {
    angularDampingOverride=0.9999
    angularSpringConstant=75
    bAngularSpring=true
    bGravityOverrideInSimSpace=true
    bOverrideAngularDamping=true
    gravityScale=0.8
    simulationSpace=BoneRelative
    alphaBoolBlend {blendOption=Linear}
    boundBone {boneName=dyn_calf_feather_l_03}
    componentAppliedLinearAccClamp=<3500 3500 3500>
    componentLinearAccScale=<1.5 1.5 1.25>
    componentLinearVelScale=<1.5 1.5 1.25>
    gravityOverride {z=-980}
    physicsBodyDefinitions [{
      sphereCollisionRadius=0
      boundBone {boneName=dyn_calf_feather_l_03}
      boxExtents=<17 17 10>
      constraintSetup {angularLimitsMax=<5 3 15>, angularLimitsMin=<-5 -15 -15>, angularTarget=<1 0 0>}
      localJointOffset=<3 1 0>
     }]
    relativeSpaceBone {boneName=root}
   }
   n35 = Anim Dynamics - Bone: dyn_calf_feather_l_04(ComponentPose=@n34, Alpha=0.5, ExternalForce="0,0,0") {
    angularDampingOverride=0.9999
    angularSpringConstant=75
    bAngularSpring=true
    bGravityOverrideInSimSpace=true
    bOverrideAngularDamping=true
    gravityScale=0.8
    simulationSpace=BoneRelative
    alphaBoolBlend {blendOption=Linear}
    boundBone {boneName=dyn_calf_feather_l_04}
    componentAppliedLinearAccClamp=<3500 3500 3500>
    componentLinearAccScale=<1.5 1.5 1.25>
    componentLinearVelScale=<1.5 1.5 1.25>
    gravityOverride {z=-980}
    physicsBodyDefinitions [{
      sphereCollisionRadius=0
      boundBone {boneName=dyn_calf_feather_l_04}
      boxExtents=<17 17 10>
      constraintSetup {angularLimitsMax=<5 3 15>, angularLimitsMin=<-5 -15 -15>, angularTarget=<1 0 0>}
      localJointOffset=<4 0 0>
     }]
    relativeSpaceBone {boneName=root}
   }
   n36 = Anim Dynamics - Bone: dyn_calf_feather_l_05(ComponentPose=@n35, Alpha=0.5, ExternalForce="0,0,0") {
    angularDampingOverride=0.9999
    angularSpringConstant=75
    bAngularSpring=true
    bGravityOverrideInSimSpace=true
    bOverrideAngularDamping=true
    gravityScale=0.8
    simulationSpace=BoneRelative
    alphaBoolBlend {blendOption=Linear}
    boundBone {boneName=dyn_calf_feather_l_05}
    componentAppliedLinearAccClamp=<3500 3500 3500>
    componentLinearAccScale=<1.5 1.5 1.25>
    componentLinearVelScale=<1.5 1.5 1.25>
    gravityOverride {z=-980}
    physicsBodyDefinitions [{
      sphereCollisionRadius=0
      boundBone {boneName=dyn_calf_feather_l_05}
      boxExtents=<17 17 10>
      constraintSetup {angularLimitsMax=<5 3 15>, angularLimitsMin=<-5 -15 -15>, angularTarget=<1 0 0>}
      localJointOffset=<3 -0.5 0>
     }]
    relativeSpaceBone {boneName=root}
   }
   n37 = Anim Dynamics - Bone: dyn_calf_feather_l_06(ComponentPose=@n36, Alpha=0.5, ExternalForce="0,0,0") {
    angularDampingOverride=0.9999
    angularSpringConstant=75
    bAngularSpring=true
    bGravityOverrideInSimSpace=true
    bOverrideAngularDamping=true
    gravityScale=0.8
    simulationSpace=BoneRelative
    alphaBoolBlend {blendOption=Linear}
    boundBone {boneName=dyn_calf_feather_l_06}
    componentAppliedLinearAccClamp=<3500 3500 3500>
    componentLinearAccScale=<1.5 1.5 1.25>
    componentLinearVelScale=<1.5 1.5 1.25>
    gravityOverride {z=-980}
    physicsBodyDefinitions [{
      sphereCollisionRadius=0
      boundBone {boneName=dyn_calf_feather_l_06}
      boxExtents=<17 17 10>
      constraintSetup {angularLimitsMax=<5 3 15>, angularLimitsMin=<-5 -15 -15>, angularTarget=<1 0 0>}
      localJointOffset=<4 0.5 0>
     }]
    relativeSpaceBone {boneName=root}
   }
   n38 = Anim Dynamics - Bone: dyn_calf_feather_l_07(ComponentPose=@n37, Alpha=0.5, ExternalForce="0,0,0") {
    angularDampingOverride=0.9999
    angularSpringConstant=75
    bAngularSpring=true
    bGravityOverrideInSimSpace=true
    bOverrideAngularDamping=true
    gravityScale=0.8
    simulationSpace=BoneRelative
    alphaBoolBlend {blendOption=Linear}
    boundBone {boneName=dyn_calf_feather_l_07}
    componentAppliedLinearAccClamp=<3500 3500 3500>
    componentLinearAccScale=<1.5 1.5 1.25>
    componentLinearVelScale=<1.5 1.5 1.25>
    gravityOverride {z=-980}
    physicsBodyDefinitions [{
      sphereCollisionRadius=0
      boundBone {boneName=dyn_calf_feather_l_07}
      boxExtents=<17 17 10>
      constraintSetup {angularLimitsMax=<5 3 15>, angularLimitsMin=<-5 -15 -15>, angularTarget=<1 0 0>}
      localJointOffset=<3 1 0>
     }]
    relativeSpaceBone {boneName=root}
   }
   n39 = Anim Dynamics - Bone: dyn_calf_feather_l_08(ComponentPose=@n38, Alpha=0.5, ExternalForce="0,0,0") {
    angularDampingOverride=0.9999
    angularSpringConstant=75
    bAngularSpring=true
    bGravityOverrideInSimSpace=true
    bOverrideAngularDamping=true
    gravityScale=0.8
    simulationSpace=BoneRelative
    alphaBoolBlend {blendOption=Linear}
    boundBone {boneName=dyn_calf_feather_l_08}
    componentAppliedLinearAccClamp=<3500 3500 3500>
    componentLinearAccScale=<1.5 1.5 1.25>
    componentLinearVelScale=<1.5 1.5 1.25>
    gravityOverride {z=-980}
    physicsBodyDefinitions [{
      sphereCollisionRadius=0
      boundBone {boneName=dyn_calf_feather_l_08}
      boxExtents=<17 17 10>
      constraintSetup {angularLimitsMax=<5 3 15>, angularLimitsMin=<-5 -15 -15>, angularTarget=<1 0 0>}
      localJointOffset=<4 0 0>
     }]
    relativeSpaceBone {boneName=root}
   }
   n40 = Anim Dynamics - Bone: dyn_calf_feather_l_09(ComponentPose=@n39, Alpha=0.5, ExternalForce="0,0,0") {
    angularDampingOverride=0.9999
    angularSpringConstant=75
    bAngularSpring=true
    bGravityOverrideInSimSpace=true
    bOverrideAngularDamping=true
    gravityScale=0.8
    simulationSpace=BoneRelative
    alphaBoolBlend {blendOption=Linear}
    boundBone {boneName=dyn_calf_feather_l_09}
    componentAppliedLinearAccClamp=<3500 3500 3500>
    componentLinearAccScale=<1.5 1.5 1.25>
    componentLinearVelScale=<1.5 1.5 1.25>
    gravityOverride {z=-980}
    physicsBodyDefinitions [{
      sphereCollisionRadius=0
      boundBone {boneName=dyn_calf_feather_l_09}
      boxExtents=<17 17 10>
      constraintSetup {angularLimitsMax=<5 3 15>, angularLimitsMin=<-5 -15 -15>, angularTarget=<1 0 0>}
      localJointOffset=<3 1 0>
     }]
    relativeSpaceBone {boneName=root}
   }
   n41 = Anim Dynamics - Bone: dyn_calf_feather_l_10(ComponentPose=@n40, Alpha=0.5, ExternalForce="0,0,0") {
    angularDampingOverride=0.9999
    angularSpringConstant=75
    bAngularSpring=true
    bGravityOverrideInSimSpace=true
    bOverrideAngularDamping=true
    gravityScale=0.8
    simulationSpace=BoneRelative
    alphaBoolBlend {blendOption=Linear}
    boundBone {boneName=dyn_calf_feather_l_10}
    componentAppliedLinearAccClamp=<3500 3500 3500>
    componentLinearAccScale=<1.5 1.5 1.25>
    componentLinearVelScale=<1.5 1.5 1.25>
    gravityOverride {z=-980}
    physicsBodyDefinitions [{
      sphereCollisionRadius=0
      boundBone {boneName=dyn_calf_feather_l_10}
      boxExtents=<17 17 10>
      constraintSetup {angularLimitsMax=<5 3 15>, angularLimitsMin=<-5 -15 -15>, angularTarget=<1 0 0>}
      localJointOffset=<4 0 0>
     }]
    relativeSpaceBone {boneName=root}
   }
   Output Pose(Result=Component To Local(ComponentPose=@n41)) {layerGroup=None}
  ]}
 EventGraph {nodes [
   Event BlueprintUpdateAnimation()
   > n1 = Sequence!(execute) -> (then_0, then_1, then_2:Exec)
   n2 = Gravity Override State Function DEPRECATED(
    execute=@n1.then_1
    CharacterPartAnimInstance=Self-Reference()
    GameTimeInSeconds=GetGameTimeInSeconds()
    GravityOverrideParams=Get Gravity Override Params()
   )
   > Set Gravity Override(Gravity Override=@n2)
   n3 = Get bIsSprinting()
   n4 = Get bIsPlayingEmote()
   > Set sprint(sprint:double="To Float (Boolean)"(InBool=OR Boolean(A=@n3, B=@n4))) -> (Output_Get:double)
   > Set emote(emote:double="To Float (Boolean)"(InBool=Get bIsPlayingEmote())) -> (Output_Get:double)
   > Set Crouch(Crouch:double="To Float (Boolean)"(InBool=Get bIsCrouching())) -> (Output_Get:double)
   > n5 = Get Joint Location Relative to Bone!(
    execute
    Target Joint:Name=dfrm_foot_l
    Relative Joint:Name=dfrm_foot_l
    Offset:Vector="1,1,1"
    SkelMeshComponent:Mesh Component Object Reference=Get Skeletal Mesh() -> (Skeletal Mesh:Skeletal Mesh Component Object Reference)
    Debug Draw:bool=false
   ) -> (then, Joint Location Vector:Vector) {is=title K2Node_CallFunction}
   > n6 = Get Joint Location Relative to Bone(Target Joint=dfrm_foot_r, Relative Joint=dfrm_foot_r, Offset="1,1,1", SkelMeshComponent=Get Skeletal Mesh(), Debug Draw=false)
   > n7 = Get Joint Location Relative to Bone(Target Joint=dyn_tail_MSTR, Relative Joint=dyn_tail_MSTR, Offset="1,1,1", SkelMeshComponent=Get Skeletal Mesh(), Debug Draw=false)
   > n8 = DistanceBetweenTwoVectors!(execute, vector1:Vector=@n5, Vector2:Vector=@n6) -> (then, Distance:double)
   > n9 = DistanceBetweenTwoVectors(vector1=@n6, Vector2=@n7)
   > n10 = Branch(Condition=Get bIsPlayingEmote())
   > Set feetCloseToTail(feetCloseToTail:double=0) -> (Output_Get:double)
   Set feetCloseToTail(
    execute=@n10.else
    feetCloseToTail=MapRangeClamped(
     Value="Min (Float)"(A:double=@n8, B:double=@n9) -> (ReturnValue:double) {is=title K2Node_CommutativeAssociativeBinaryOperator}
     InRangeA=20
     InRangeB=5
    )
   )
   n11 = Get bIsSkyDiving()
   n12 = Get bIsPlayingEmote()
   n13 = Branch(execute=@n1.then_2, Condition=AND Boolean!(A=Get bIsInFrontEnd(), B=NOT Boolean(A=@n12), C:bool=NOT Boolean(A=@n11)) -> (ReturnValue))
   > Set frontend(frontend:double=1) -> (Output_Get:double)
   Set frontend(execute=@n13.else, frontend=0)
   Event OnCharacterCustomizationLoaded()
   > Set Skeletal Mesh(Skeletal Mesh:Skeletal Mesh Component Object Reference=GetPartSkeletalMeshComponent()) -> (Output_Get:Skeletal Mesh Component Object Reference)
  ]}
]
variables [
 Crouch {type=number}
 emote {type=number}
 feetCloseToTail {type=number}
 frontend {type=number}
 Gravity Override {type=Vector}
 Gravity Override Params {type=GravityOverrideParamsStruct}
 is Custom Emote Playing {type=boolean}
 Skeletal Mesh {type=SkeletalMeshComponent}
 sprint {type=number}
]
